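# Unlit translucent particle material for the GhostPortal backpack effect, drawing a wavering ghost on sprites.
# A scrolling normal map, scaled by a ripple mask and the Distort dynamic parameter, warps axis-swizzled UVs sampling T_Ghost; dynamic Time drives the scroll.
# The sample is desaturated, contrast-adjusted and tinted by particle color.
# Its alpha combines particle alpha, vertical falloff, sine edge masks, view-facing fade, depth fade and near-camera fade.
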Material /Game/Accessories/FORT_Backpacks/M_MED_GhostPortal_Backpack/FX/M_GhostSubUV
graph {
 properties {
  bAutomaticallySetUsageInEditor=false
  bUsedWithNiagaraSprites=true
  bUsedWithParticleSprites=true
  bUsedWithStaticMesh=false
  blendMode=BLEND_Translucent
  pixelDepthOffsetMode=PDOM_Legacy
  shadingModel=MSM_Unlit
  usedShadingModels=Unlit
 }
 nodes [
  n1 = ParticleColor()
  n2 = TextureCoordinate()
  n3 = TextureCoordinate() {uTiling=6, vTiling=0.55}
  n4 = DynamicParameter() {paramNames [SubUV,Param2,Distort,Time]}
  n5 = ComponentMask(None=TextureSample(UVs=Panner(Coordinate=@n3, Time=@n4.Time) {speedY=0.2}) {samplerType=SAMPLERTYPE_Normal, texture=&T_Weap_Distort_N_02}.RGB) {g=true, r=true}
  n6 = Multiply(
   A=@n4.Distort
   B=TextureSample(UVs=TextureCoordinate() {uTiling=6}) {samplerType=SAMPLERTYPE_LinearColor, texture=/Game/Effects/Fort_Effects/Effects/Water/Textures/WaterDecalRippleMask}.R
  )
  n7 = TextureSample(UVs=Add(
    A=MaterialFunctionCall(XYZ=@n2, XY=@n2) {materialFunction=&Swizzle}.YX
    B=Multiply(A=@n5, B=@n6)
   )) {texture=&/T_Ghost}
  n8 = Multiply(
   A=@n1.RGB
   B=MaterialFunctionCall(In=Desaturation(None=@n7.RGB), Contrast=ScalarParameter() {parameterName=Contrast}) {materialFunction=&CheapContrast}.Result
  )
  MaterialFunctionCall(Emissive=@n8) {materialFunction=&MF_FX_EmissiveAdjustments}
  n9 = OneMinus(None=ComponentMask(None=TextureCoordinate()) {g=true})
  n10 = Sine(None=TextureCoordinate() {uTiling=0.5, vTiling=0.5})
  n11 = ComponentMask(None=@n10) {r=true}
  n12 = ComponentMask(None=@n10) {g=true}
  n13 = Multiply(A=Multiply(A=@n1.A, B=@n7.A), B=Multiply(A=Multiply(A=@n9, B=@n9), B=Multiply(A=@n11, B=@n12)))
  n14 = Abs(None=DotProduct(A=CameraVectorWS(), B=VertexNormalWS()))
  Multiply(
   A=DepthFade(Opacity=Multiply(A=@n13, B=@n14)) {fadeDistanceDefault=60}
   B=MaterialFunctionCall(Distance=ScalarParameter() {defaultValue=60, parameterName=NearFadeDistance}) {materialFunction=&Near_Camera_Fade}.Result
  )
  Reroute(None=@n8)
 ]
}
