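# Translucent, unlit, two-sided mesh-particle material for a swirling vortex effect on the Ghost Portal backpack.
# Particle-driven time pans noise that distorts further noise samples, which blend two cyan HDR colors and form a UV-edge-masked shape eroded by particle alpha.
# View-angle and depth fades shape opacity and refraction, and scene color modulates the glow, with cheaper mobile fallbacks.
# A UseParticleColor switch blends in particle color and alpha, and a camera offset shifts vertices.
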Material /Game/Accessories/FORT_Backpacks/M_MED_GhostPortal_Backpack/FX/M_MeshVotex
graph {
 properties {
  bUsedWithMeshParticles=true
  bUsedWithNiagaraMeshParticles=true
  blendMode=BLEND_Translucent
  pixelDepthOffsetMode=PDOM_Legacy
  refractionMethod=RM_IndexOfRefraction
  shadingModel=MSM_Unlit
  twoSided=true
  usedShadingModels=Unlit
 }
 nodes [
  n1 = Sine(None=TextureCoordinate() {uTiling=0.5, vTiling=0.5})
  n2 = DynamicParameter() {paramNames [Time,Random,Param3,Param4]}
  n3 = Add(A=@n2.Time, B=@n2.Time)
  n4 = Panner(Coordinate=TextureCoordinate() {vTiling=0.5}, Time=@n3) {speedX=-0.2, speedY=-0.2}
  n5 = TextureSample(UVs=Panner(Coordinate=TextureCoordinate() {vTiling=2}, Time=@n3) {speedX=-0.5, speedY=1}) {texture=&T_NoisePack_01}
  n6 = Multiply(A=AppendVector(A=@n5.G, B=@n5.G)) {constB=0.2}
  n7 = Add(A=Subtract(A=@n4) {constB=-0.1}, B=@n6)
  n8 = Multiply(
   A=Multiply(A=ComponentMask(None=@n1) {r=true}, B=ComponentMask(None=@n1) {g=true})
   B=Multiply(A=TextureSample(UVs=@n7) {texture=&T_NoisePack_01}.R) {constB=2}
  )
  n9 = Subtract(A=@n8, B=ParticleColor().A)
  n10 = DotProduct(A=CameraVectorWS(), B=VertexNormalWS())
  n11 = ShadingPathSwitch(Default=Clamp(None=Abs(None=@n10)) {maxDefault=0.5}, Mobile=Constant() {r=0.5}) {desc=Migrated from FeatureLevelSwitch}
  n12 = DepthFade() {fadeDistanceDefault=3}
  n13 = Constant() {r=1}
  n14 = Saturate(None=Multiply(A=Multiply(A=@n9, B=@n11), B=ShadingPathSwitch(Default=@n12, Mobile=@n13) {desc=Migrated from FeatureLevelSwitch}))
  LinearInterpolate(B=ScalarParameter() {defaultValue=0.8, parameterName=Ref}, Alpha=@n14) {constA=1}
  LinearInterpolate(A=@n14, B=Multiply(A=ParticleColor().A, B=@n8), Alpha=ScalarParameter() {parameterName=UseParticleColor})
  MaterialFunctionCall(Offset Amount=Constant() {r=8}) {materialFunction=/Engine/Functions/Engine_MaterialFunctions02/WorldPositionOffset/CameraOffset}
  n15 = LinearInterpolate(A=SceneColor!(UVs).RGB) {constAlpha=0.3, constB=0.5}
  n16 = ScalarParameter() {parameterName=UseParticleColor}
  n17 = Add(A=@n4, B=@n6)
  n18 = LinearInterpolate(
   A=LinearInterpolate(
    A=VectorParameter() {parameterName=Lerp A, defaultValue=<rgba 0 2.5 4 1>}.RGB
    B=VectorParameter() {parameterName=Lerp B, defaultValue=<rgba 0 10.2419 12 1>}.RGB
    Alpha=TextureSample(UVs=@n17) {texture=&T_NoisePack_01}.R
   )
   B=Multiply(A=ParticleColor().RGB)
   Alpha=ScalarParameter() {parameterName=UseParticleColor}
  )
  n19 = ShadingPathSwitch(Default=Multiply(A=LinearInterpolate(A=@n15, Alpha=@n16), B=@n18), Mobile=Multiply(A=@n18) {constB=0.5}) {desc=Migrated from FeatureLevelSwitch}
  MaterialFunctionCall(Emissive=Multiply(A=@n19, B=@n19)) {materialFunction=&MF_FX_EmissiveAdjustments}
 ]
}
