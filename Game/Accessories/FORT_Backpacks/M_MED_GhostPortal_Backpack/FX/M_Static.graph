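# Translucent, unlit, two-sided mesh-particle material for the Ghost Portal backpack effect, drawing flickering, static-like trails tinted by a Color parameter.
# Vertex math widens the mesh perpendicular to the view by Scale and noise, adds noise wobble growing along V, and bends it by dynamic OffsetPosX.
# Emissive scrolls a smoke-trail texture, brightened toward low V, with time offset by particle red and Random.
# Opacity combines noise, particle alpha, end fades, depth fade and a green-driven ramp.
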Material /Game/Accessories/FORT_Backpacks/M_MED_GhostPortal_Backpack/FX/M_Static
graph {
 properties {
  bUsedWithMeshParticles=true
  bUsedWithNiagaraMeshParticles=true
  blendMode=BLEND_Translucent
  pixelDepthOffsetMode=PDOM_Legacy
  shadingModel=MSM_Unlit
  twoSided=true
  usedShadingModels=Unlit
 }
 nodes [
  n1 = ComponentMask(None=TextureCoordinate() {uTiling=2}) {r=true}
  n2 = Add(A=ParticleColor().R, B=DynamicParameter() {paramNames [OffsetPosX,OffsetPosY,OffsetPosZ,Random]}.Random)
  n3 = Panner(Coordinate=TextureCoordinate() {coordinateIndex=1, vTiling=0}, Time=@n2) {speedX=0.03}
  n4 = Clamp(None=ConstantBiasScale(None=TextureSample(UVs=@n3) {samplerType=SAMPLERTYPE_LinearColor, texture=/Game/Global/Textures/Default/Noise/Noise32x32}.R) {bias=-0.5, scale=2})
  n5 = Multiply(
   A=Multiply(A=Subtract(A=@n1), B=ScalarParameter() {defaultValue=8, parameterName=Scale})
   B=ShadingPathSwitch(Default=@n4, Mobile=@n4) {desc=Migrated from FeatureLevelSwitch}
  )
  n6 = TextureCoordinate() {coordinateIndex=1, vTiling=0.3}
  n7 = ComponentMask(None=@n6) {r=true}
  n8 = ComponentMask(None=@n6) {g=true}
  n9 = Multiply(A=@n8, B=@n8)
  n10 = TextureSample(UVs=Panner(Coordinate=AppendVector(A=@n7, B=@n9)) {speedX=0.02, speedY=0.05}) {samplerType=SAMPLERTYPE_LinearColor, texture=&HighContrastNoise}
  n11 = TextureCoordinate()
  n12 = Multiply(A=Multiply(A=ConstantBiasScale(None=@n10.RGB) {bias=-0.5, scale=2}, B=ComponentMask(None=@n11) {g=true})) {constB=24}
  n13 = WorldPosition()
  n14 = Normalize(VectorInput=Subtract(A=CameraPositionWS(), B=Subtract(A=@n12, B=@n13.XYZ)))
  n15 = CrossProduct(
   A=Transform(None=Constant3Vector() {constant=<rgba 1 0 0 0>})
   B=ShadingPathSwitch(Default=@n14, Mobile=@n14) {desc=Migrated from FeatureLevelSwitch}
  )
  n16 = DynamicParameter() {paramNames [OffsetPosX,OffsetPosY,OffsetPosZ,Random]}
  n17 = ComponentMask(None=TextureCoordinate()) {g=true}
  Add(
   A=Add(A=Multiply(A=@n5, B=@n15) {constB=50}, B=ShadingPathSwitch(Default=@n12, Mobile=@n12) {desc=Migrated from FeatureLevelSwitch})
   B=Multiply(
    A=MaterialFunctionCall(X=@n16.OffsetPosX, Y=@n16.OffsetPosX, Z=@n16.OffsetPosX) {materialFunction=&MakeFloat3}.Result
    B=Multiply(A=@n17, B=@n17)
   )
  )
  Constant() {r=1}
  Constant()
  n18 = ComponentMask(None=TextureCoordinate()) {g=true}
  n19 = OneMinus(None=@n18)
  n20 = Multiply(A=Saturate(None=Subtract(A=@n19) {constB=0.5})) {constB=5}
  n21 = Multiply(A=@n20, B=@n20)
  n22 = OneMinus(None=Multiply(A=@n18, B=@n18))
  n23 = Add(A=Multiply(A=@n21, B=@n21), B=@n22)
  n24 = MaterialFunctionCall(Float2=TextureCoordinate() {vTiling=3}) {materialFunction=&BreakOutFloat2Components}
  n25 = TextureSample(UVs=Panner(Coordinate=AppendVector(A=@n24.G, B=@n24.G), Time=@n2) {speedX=-1.5}) {texture=/Game/Effects/Fort_Effects/Textures/Smoke/T_Smoke_Trail_Soft_Chromatic}
  MaterialFunctionCall(Emissive=Multiply(A=Multiply(A=@n23, B=@n25.RGB), B=VectorParameter() {parameterName=Color, defaultValue=<rgba 0.1137 0.5631 1 1>}.RGB)) {materialFunction=&MF_FX_EmissiveAdjustments}
  n26 = Multiply(A=Multiply(A=@n22, B=@n25.G), B=@n4)
  n27 = ParticleColor()
  n28 = TextureCoordinate()
  n29 = Subtract(A=OneMinus(None=ComponentMask(None=@n28) {g=true})) {constB=0.85}
  n30 = OneMinus(None=Saturate(None=Multiply(A=@n29) {constB=8}))
  n31 = ShadingPathSwitch(Default=DepthFade() {fadeDistanceDefault=2}, Mobile=Constant() {r=1}) {desc=Migrated from FeatureLevelSwitch}
  n32 = ParticleColor()
  n33 = Multiply(A=ComponentMask(None=TextureCoordinate()) {g=true}) {constB=0.75}
  Multiply(
   A=Multiply(A=Multiply(A=@n26, B=@n27.A) {constB=0.5}, B=Multiply(A=@n30, B=@n31))
   B=TextureSample(UVs=Add(A=@n32.G, B=@n33)) {samplerSource=SSM_Clamp_WorldGroupSettings, texture=/Game/Effects/Fort_Effects/Textures/Ramps/T_Ramp_Smooth}.R
  )
  Transform(None=Constant3Vector() {constant=<rgba 0 0 1 0>}) {transformSourceType=TRANSFORMSOURCE_Camera}
 ]
}
